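# Sound cue for the Apollo Tinfoil beacon prop, named for its ping-detect sound.
# A single wave player feeds the output directly, reusing the character Interrogate sweep sample.
# The cue boosts volume, routes through the player foley sound class, and overrides attenuation for long-range spatialized falloff with a wide stereo spread.
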SoundCue /Game/Athena/Sounds/Props/Tinfoil/Apollo_Tinfoil_Beacon_PingDetect_Cue
root=WavePlayer_0
nodes [{id=WavePlayer_0, properties {soundWaveAssetPtr=/Game/Athena/Sounds/Character/Interrogate/Interrogate_Sweep_01}}]
properties {
 bOverrideAttenuation=true
 maxDistance=10000
 soundClassObject=&Fort_Player_Foley
 volumeMultiplier=2
 attenuationOverrides {falloffDistance=9000, stereoSpread=1000, attenuationShapeExtents {x=1000}}
}
